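# Physics asset for the InflateMate guitar pickaxe cosmetic, defining bodies and joints for its handle and neck bones.
# Four capsule bodies form a chain from handle through neck_01, neck_02 and neck_03, with the neck capsules offset and rotated on their bones.
# Three constraints link neighboring bodies with twist locked and very small swing limit angles.
# The two joints nearest the handle each also lock one swing axis, on opposite axes.
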PhysicsAsset /SparksCosmetics/Instrument/Guitar/InflateMate/Meshes/InflateMate_Guitar_Pickaxe_Physics
bodies {
 handle [{shapeType=Capsule, length=10, radius=4}]
 neck_01 [{shapeType=Capsule, length=10.5977, radius=3.3643, shapeName=neck_01_capsule, center=<8 0 0>, rotation=<rot 0 90 -89.797>}]
 neck_02 [{shapeType=Capsule, length=19.3065, radius=2.9999, shapeName=neck_02_capsule, center=<13 0 0>, rotation=<rot 0 89.9912 -88.587>}]
 neck_03 [{shapeType=Capsule, length=10.5211, radius=2.9163, shapeName=neck_03_capsule, center=<8 0 0>, rotation=<rot 0 89.7552 -87.3919>}]
}
constraints [
 {
  bone1Name=neck_01
  bone2Name=handle
  swing1LimitDegrees=1
  swing2LimitDegrees=2
  swing2Motion=Locked
  twistLimitDegrees=45
  twistMotion=Locked
 }
 {
  bone1Name=neck_02
  bone2Name=neck_01
  swing1LimitDegrees=0.5
  swing1Motion=Locked
  swing2LimitDegrees=0.5
  twistLimitDegrees=45
  twistMotion=Locked
 }
 {bone1Name=neck_03, bone2Name=neck_02, swing1LimitDegrees=1, swing2LimitDegrees=1, twistLimitDegrees=45, twistMotion=Locked}
]
